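# Character material function that adjusts base color, metallic, specular and roughness of incoming material attributes within a masked area, toggled by a static switch.
# Base color lerps by an amount toward a tint color, either multiplied in or replacing it.
# Metallic, specular and roughness biases are added to the originals or, with OverrideSRM, replace them, then saturated.
# A channel-mask parameter on the Mask input drives the blend between original and adjusted attributes.
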MaterialFunction /Game/Global/Materials/Character_Parent_2023/ParameterAdjustment/MF_ParameterAdjustment
graph {nodes [
  n1 = NamedRerouteUsage()
  n2 = SetMaterialAttributes!(
   MaterialAttributes=@n1
   BaseColor=NamedRerouteUsage()
   Metallic=NamedRerouteUsage()
   Specular=NamedRerouteUsage()
   Roughness=NamedRerouteUsage()
  ) {attributeSetTypes [
    #1
    #2
    #3
    #4
   ]}
  n3 = NamedRerouteUsage()
  FunctionOutput(None=StaticSwitchParameter(True=BlendMaterialAttributes(A=@n1, B=@n2, Alpha=@n3), False=NamedRerouteUsage()) {defaultValue=true, group=ParameterAdjustment, parameterName=UseParameterAdjustment, sortPriority=0})
  NamedRerouteDeclaration(None=ChannelMaskParameter(None=FunctionInput() {bUsePreviewValueAsDefault=true, inputName=Mask, sortPriority=1}) {group=ParameterAdjustment, parameterName=ParameterAdjustmentMaskChannel, sortPriority=1}.R) {name=IN_Mask, nodeColor=<rgba 1 0 0.6094 1>}
  n4 = ScalarParameter() {group=ParameterAdjustment, parameterName=SpecularBias, sortPriority=21}
  n5 = Add(A=@n4, B=NamedRerouteUsage())
  n6 = NamedRerouteUsage()
  NamedRerouteDeclaration(None=Saturate(None=StaticSwitch(True=@n4, False=@n5, Value=@n6))) {name=OUT_Specular, nodeColor=<rgba 0.8516 1 0 1>}
  n7 = GetMaterialAttributes(None=NamedRerouteDeclaration(None=FunctionInput() {inputName=Material, inputType=FunctionInput_MaterialAttributes}) {name=IN_Material, nodeColor=<rgba 1 0 0.6094 1>}) {attributeGetTypes [
    #1
    #2
    #3
    #4
   ]}
  NamedRerouteDeclaration(None=@n7.BaseColor) {name=IN_BaseColor, nodeColor=<rgba 1 0.75 0 1>}
  NamedRerouteDeclaration(None=@n7.Roughness) {name=IN_Roughness, nodeColor=<rgba 0.4297 0 1 1>}
  NamedRerouteDeclaration(None=@n7.Metallic) {name=IN_Metallic, nodeColor=<rgba 0 1 0.1094 1>}
  NamedRerouteDeclaration(None=@n7.Specular) {name=IN_Specular, nodeColor=<rgba 1 0 0.2109 1>}
  n8 = NamedRerouteUsage()
  n9 = VectorParameter() {group=ParameterAdjustment, parameterName=BaseColorAdjustment, sortPriority=10, defaultValue=<rgba 1 1 1 1>}
  n10 = Multiply(A=@n8, B=@n9.RGB)
  n11 = ScalarParameter() {group=ParameterAdjustment, parameterName=BaseColorAdjustmentAmount, sliderMax=1, sortPriority=11}
  NamedRerouteDeclaration(None=LinearInterpolate(
    A=@n8
    B=StaticSwitchParameter(True=@n10, False=@n9.RGB) {group=ParameterAdjustment, parameterName=MultiplyBaseColor, sortPriority=12}
    Alpha=Saturate(None=@n11)
   )) {name=OUT_BaseColor, nodeColor=<rgba 0 0.5312 1 1>}
  n12 = ScalarParameter() {group=ParameterAdjustment, parameterName=MetallicBias, sortPriority=22}
  n13 = Add(A=@n12, B=NamedRerouteUsage())
  n14 = NamedRerouteUsage()
  NamedRerouteDeclaration(None=Saturate(None=StaticSwitch(True=@n12, False=@n13, Value=@n14))) {name=OUT_Metallic, nodeColor=<rgba 0.8516 1 0 1>}
  NamedRerouteDeclaration(None=StaticBoolParameter() {group=ParameterAdjustment, parameterName=OverrideSRM, sortPriority=30}) {name=OverrideSRM, nodeColor=<rgba 0.8281 0 1 1>}
  n15 = ScalarParameter() {group=ParameterAdjustment, parameterName=RoughnessBias, sortPriority=20}
  n16 = Add(A=@n15, B=NamedRerouteUsage())
  n17 = NamedRerouteUsage()
  NamedRerouteDeclaration(None=Saturate(None=StaticSwitch(True=@n15, False=@n16, Value=@n17))) {name=OUT_Roughness, nodeColor=<rgba 0.8516 1 0 1>}
 ]}
